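; BGA_IslandPortal is the Creative island portal rift that teleports players and vehicles to islands or the hub and opens island-selection UI.
; Overlap or interaction starts teleports, with rift FX, audio and optional skydiving, delayed through minigame events when needed.
; UpdatePortal drives rift timelines, gold visuals and a beacon for the local player's portal, and info plates.
; Interaction prompts depend on code-screen and look-at checks; UI classes load asynchronously.
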(event EventBeginPlay
  (bind _teleportheight (Variables|RiftPortal|GetTeleportHeight))
  (bind _returnvalue (ScalableFloat|GetValueatLevel _teleportheight))
  ;[1 1 1] EVENT--> BeginPlay
  (|CallParentFunction)
  (Variables|FortAthenaCreativePortal|SetInteractIconPlacementCapsule (Variables|Default|GetOverlapCapsule))
  (CallFunction|SetOverlapComponent (Variables|Default|GetOverlapCapsule))
  (Utilities|FlowControl|Sequence
    (:then_0
      (Networking|CanExecuteCosmeticEvents
        (:True
          ;;[0 0 0] load spawned ui classes
          (bind _class (Utilities|AsyncLoadClassAsset (select (CallFunction|ShouldUseMultipleIPUI) (Variables|UI|GetUI_MultipleIP_IslandSelectSoftClass) (Variables|UI|GetUI_IslandSelectSoftClass)))
            (:then
              _)
            (:Completed
              (bind _asuser_widget (Utilities|Casting|CastToUserWidgetClass _class))
              (Variables|UI|SetUI_IslandSelectClass _asuser_widget))))
        (:False
          (if (not (> (ScalableFloat|GetValueatLevel (Variables|Default|GetShouldSpawnRift?)) 0.0))
            (Actor|DestroyActor))))
      (Utilities|IsValid (Variables|FortAthenaCreativePortal|GetDestinationActor)
        (:Is%20Valid
          (bind _asbga_island_portal (Utilities|Casting|CastToBGA_IslandPortal (Variables|FortAthenaCreativePortal|GetDestinationActor))
            (:then
              (bind _location (CallFunction|GetPortalPosition _asbga_island_portal))
              (Variables|SetTeleportLocation_X (Math|Vector|MakeVector (.x _location) (.y _location) (ScalableFloat|GetValueatLevel (Variables|RiftPortal|GetStarterIslandTeleportHeight)))))
            (:CastFailed
              (Variables|SetTeleportLocation_X (Math|Vector|MakeVector (.x (+ (Transformation|GetActorLocation (Variables|FortAthenaCreativePortal|GetDestinationActor)) (* (Transformation|GetActorForwardVector (Variables|FortAthenaCreativePortal|GetDestinationActor)) (ScalableFloat|GetValueatLevel (Variables|Portal|GetTeleportExitDistance))))) (.y (+ (Transformation|GetActorLocation (Variables|FortAthenaCreativePortal|GetDestinationActor)) (* (Transformation|GetActorForwardVector (Variables|FortAthenaCreativePortal|GetDestinationActor)) (ScalableFloat|GetValueatLevel (Variables|Portal|GetTeleportExitDistance))))) _returnvalue)))))
        (:Is%20Not%20Valid
          (Variables|SetTeleportLocation_X (Math|Vector|MakeVector (.x (+ (Transformation|GetActorLocation) (* (Transformation|GetActorForwardVector) (ScalableFloat|GetValueatLevel (Variables|Portal|GetTeleportExitDistance))))) (.y (+ (Transformation|GetActorLocation) (* (Transformation|GetActorForwardVector) (ScalableFloat|GetValueatLevel (Variables|Portal|GetTeleportExitDistance))))) _returnvalue)))))
    (:then_1
      (CallFunction|SetInteractionInteractComponent (Variables|Default|GetCapsule))
      (if (Variables|FortAthenaCreativePortal|GetbReturnToCreativeHub)
        (Audio|Components|Audio|SetVolumeMultiplier (Variables|Sound|GetAudio_Looping) 0.5 :!class /Script/Engine.AudioComponent)
        (Components|Sphere|SetSphereRadius (Variables|Default|GetPostProcessRadius) 25.0 false)
        (Variables|UI|SetUIShowDistance 1e+03)))
    (:then_2
      (Networking|CanExecuteCosmeticEvents
        (:True
          (UI|UpdateInfoPlatePointer)
          ;[0 0 0] setup soft widget ref and init through interface
          (Utilities|FlowControl|Sequence
            (:then_0
              (bind _class_2 (Utilities|AsyncLoadClassAsset (Variables|UI|GetUI_InteractionDisplayWidgetSoftClass))
                (:Completed
                  (bind _asuser_widget_2 (Utilities|Casting|CastToUserWidgetClass _class_2))
                  (bind _returnvalue_2 (UserInterface|CreateWidget _asuser_widget_2 (Game|GetPlayerController 0)))
                  (CallFunction|SetInteractionDisplayWidget _returnvalue_2)
                  (Init|InitfromObject(Message) _returnvalue_2 self)
                  ;;[0 0 0] update ui
                  (CallFunction|UpdatePopulationDisplay)
                  (CallFunction|UpdateNameDisplay))))
            (:then_1
              (bind _class_3 (Utilities|AsyncLoadClassAsset (Variables|UI|GetUI_PortalInfoPlateSoftClass))
                (:Completed
                  (bind _asuser_widget_3 (Utilities|Casting|CastToUserWidgetClass _class_3))
                  (bind _returnvalue_3 (UserInterface|CreateWidget _asuser_widget_3 (Game|GetPlayerController 0)))
                  (UserInterface|SetWidget (Variables|Default|GetUI_PortalInfoPlate) _returnvalue_3)
                  (Init|InitfromObject(Message) _returnvalue_3 self)
                  (CallFunction|CheckShouldDisplayUI)
                  ;[0 0 0] timer for when to show/hide ui - wip moving to overlap based
                  (Utilities|FlowControl|Delay (Math|Random|RandomFloatinRange 0.0 0.2))
                  (bind _returnvalue_4 (Utilities|Time|SetTimerbyEvent (EventDispatchers|CreateEvent :!event Custom|CheckShouldDisplayUI) 0.4 true))
                  (Variables|UI|SetUIDistanceCheckTimer _returnvalue_4)
                  (CallFunction|UpdatePortal))))))
        (:False
          ;[1 1 1] Call UpdatePortal on the server to make sure all collision settings are updated on the initial load
          (CallFunction|UpdatePortal))))))

(event Custom|Play%20Teleport%20Audio
  (bind _overlapactor (Variables|RiftPortal|GetOverlapActor))
  (bind _playerpawn (Teleport|GetPlayerPawnfromActor _overlapactor))
  ;[1 1 1] Play Sound when teleporting
  (Utilities|IsValid _playerpawn
    (:Is%20Valid
      (bind _instigator (Variables|Actor|GetInstigator _playerpawn)
        (:then
          (if (Pawn|IsLocallyControlled _instigator)
            (Audio|PushSoundMixModifier "/Game/Sounds/Creative/MixModifiers/Creative_InGameStart_MixMod.Creative_InGameStart_MixMod")
            (Utilities|FlowControl|Delay 0.5)
            (Audio|PlaySound2D "/Game/Sounds/Fort_GamePlay_Sounds/Player/RiftPortal/RiftPortal_Creative_Teleport_Relocate_Cue.RiftPortal_Creative_Teleport_Relocate_Cue")))
        (:else)))))

(event Custom|PlayTeleportTransition
  (params (DestinationLocation :type Vector) (From%20Lobby :type Boolean))
  (Utilities|FlowControl|Sequence
    (:then_0
      (if From%20Lobby
        (Audio|SpawnSoundatLocation (Variables|Sound|GetPlayerExitSkyCrack) DestinationLocation)
        (UI|CloseandClearDisplayedWidget)
        (UI|SetPortalUIVisibility false)))
    (:then_1
      _)
    (:then_2
      ;[1 1 1] FX for Big Sky crack can only play every 6 seconds
      (Utilities|FlowControl|Delay 6.0)
      (goto _label Reset)))
  (Utilities|FlowControl|DoOnce false :!label _label
    (:Completed
      (Utilities|FlowControl|Delay 1.0)
      (bind _returnvalue (Effects|Components|ParticleSystem|SpawnFXAtLocation (Variables|VFX|GetCrackExitRift) (+ DestinationLocation "0, 0,200.000000") (Math|Rotator|MakeRotator 0.0 0.0 (Math|Random|RandomFloatinRange -1.8e+02 1.8e+02)) "1.000000,1.000000,1.000000" true "None" true))
      (LOD|SetMaxDrawDistance _returnvalue 0.0))))

(event Custom|PlayTeleportFX
  (params (PlayerLocation :type Vector) (DestinationLocation :type Vector))
  ;[1 1 1] Spawn teleport visuals and audio for clients
  (if (not (Variables|FortAthenaCreativePortal|GetbLocalClientLacksPermission))
    ;[1 1 1] Play Sound when teleporting
    (bind _returnvalue (Effects|Components|ParticleSystem|SpawnFXAtLocation (Variables|VFX|GetEnterRift_NS) PlayerLocation))
    (if (Variables|Default|GetIsMyPortal)
      (bind _asniagara_particle_system_component (Utilities|Casting|CastToNiagaraComponent _returnvalue))
      (Niagara|NiagaraSetColorArray _asniagara_particle_system_component "Color Array" (Variables|VFX|GetNS%20Color%20Array%20Data)))
    (Audio|PlaySoundatLocation (Variables|Sound|GetEnteredRift) PlayerLocation)))

(event Custom|CheckShouldDisplayUI
  (bind _ui_portalinfoplate (Variables|Default|GetUI_PortalInfoPlate))
  (bind _returnvalue (UserInterface|GetUserWidgetObject _ui_portalinfoplate))
  ;[0 0 0] Show or hide the portal ui info plate
  (if (Math|Boolean|ORBoolean (Variables|Portal|GetbPortalOpen) (Variables|Default|GetIsMyPortal) (Variables|UI|GetbIsBeingDisplayed) (Variables|Default|GetbAlwaysDisplay) (Variables|Portal|GetbUserInitiatedLoad))
    (Utilities|IsValid _returnvalue
      (:Is%20Valid
        (bind _asfort_player_controller (Utilities|Casting|CastToFortPlayerController (Widget|GetOwningPlayer _returnvalue)))
        (Utilities|IsValid (Game|Player|GetPlayerPawn _asfort_player_controller)
          (:Is%20Valid
            (bind _distance (Class|MissionBlueprintFunctionLibrary|DistanceBetweenTwoVectors (Transformation|GetWorldLocation (Variables|/Script/Engine.Character|GetMesh (Game|Player|GetPlayerPawn _asfort_player_controller))) (Transformation|GetWorldLocation (Variables|Default|GetUI_PortalInfoPlate))))
            (UI|SetPortalUIVisibility (< _distance (Variables|UI|GetUIShowDistance)) _distance)))))))

(event EventBP_OnOverlapBegin
  (params (OtherActor :type Actor%20Object%20Reference))
  ;[1 1 1] Teleport Player and/or Vehicle
  (if (not (CodeOverride|PortalTeleportsbyInteraction))
    (Teleport|StartTeleport OtherActor)))

(event Custom|LoadIslandUI
  (params (InteractingPawn :type Fort%20Player%20Pawn%20Object%20Reference))
  (bind _ui_islandselectclass (Variables|UI|GetUI_IslandSelectClass))
  ;[0 0 0] Open Island UI
  (Networking|CanExecuteCosmeticEvents
    (:True
      (Utilities|IsValid (Variables|UI|GetDisplayedWidget)
        (:Is%20Not%20Valid
          (if (Utilities|IsValidClass _ui_islandselectclass)
            (bind _returnvalue (UserInterface|CreateWidget _ui_islandselectclass (Fortnite|GetFortPlayerControllerfromActor InteractingPawn)))
            (Init|InitfromObject(Message) _returnvalue self)
            (Variables|UI|SetDisplayedWidget _returnvalue)))))))

(event EventOverlappingOnBeginPlay
  (params (Actor :type Actor%20Object%20Reference))
  ;[1 1 1] Teleport Player and/or Vehicle
  (if (not (CodeOverride|PortalTeleportsbyInteraction))
    (Teleport|StartTeleport Actor)))

(event Custom|TeleportToCreativeHub
  (bind _overlapactor (Variables|RiftPortal|GetOverlapActor))
  (bind _playerpawn (Teleport|GetPlayerPawnfromActor _overlapactor))
  ;[1 1 1] Teleport Player and/or Vehicle
  (if (Networking|HasAuthority)
    (Utilities|IsValid _playerpawn
      (:Is%20Valid
        (bind _asfort_player_controller_athena (Utilities|Casting|CastToFortPlayerControllerAthena (Pawn|GetController _playerpawn)))
        (CallFunction|PlayTeleportAudio)
        (Class|FortPlayerControllerAthena|TeleporttoPlaygroundLobbyIsland _asfort_player_controller_athena)))))

(event Custom|CheckIsMyPortal
  (params (IsMyPortal :type Boolean))
  ;[1 1 1] My Portal
  (Variables|Default|SetIsMyPortal IsMyPortal))

(event EventOwningPlayerChanged
  ;[0 0 0] wip nameplate setting
  (CallFunction|UpdatePortal))

(event EventNotifyTeleportedVehicle
  (params (VehicleActor :type Actor%20Object%20Reference))
  ;[1 1 1] Teleport Player and/or Vehicle
  (bind _asfort_vehicle_interface (Utilities|Casting|CastToFortVehicleInterface (stale _output_get)))
  (bind _output_get (Variables|RiftPortal|SetVehicleActor VehicleActor))
  (CallFunction|PlayTeleportFX (Class|FortVehicleInterface|GetOwnerLocation _asfort_vehicle_interface) (Variables|Portal|GetTeleportLocation))
  (bind _output_get_2 (Variables|RiftPortal|SetPassengerArray (Vehicle|GetAllPassengers _asfort_vehicle_interface)))
  (for _array_element _output_get_2
    (RemoteControlPawn|ForceDetachfromRemoteControlledPawn _array_element)
    (Ability|SendGameplayEventtoActor _array_element "(TagName=\"Abilities.Generic.Rift.Trigger\")")
    (Variables|Portal|SetTeleportLocation (+ (Variables|Portal|GetTeleportLocation) "0,100.000000, 0"))))

(event EventBroadcastIsBeingInteractedWith
  (params (IsInteracting :type Boolean))
  ;[1 1 1] Interaction Interface
  (Default|CallIsBeingInteractedWith IsInteracting))

(event Custom|UpdatePortal
  (bind _ui_portalinfoplate (Variables|Default|GetUI_PortalInfoPlate))
  (bind _returnvalue (UserInterface|GetUserWidgetObject _ui_portalinfoplate))
  (bind _sm_rift (Variables|Default|GetSM_Rift))
  (bind _sm_rift_2 (Variables|Default|GetSM_Rift))
  (bind _p_riftclosingup (Variables|VFX|GetP_RiftClosingUp))
  (bind _p_rift_idle_loop_gold (Variables|Default|GetP_Rift_Idle_Loop_Gold))
  (bind _ns_rift_idle_loop (Variables|Default|GetNS_Rift_Idle_Loop))
  (Utilities|FlowControl|Sequence
    (:then_0
      (Networking|CanExecuteCosmeticEvents
        (:True
          (Utilities|FlowControl|Sequence
            (:then_0
              (Utilities|IsValid _returnvalue
                (:Is%20Valid
                  (bind _returnvalue_2 (CallFunction|GetPlayerState))
                  (CallFunction|CheckIsMyPortal (Utilities|Equal(Object) _returnvalue_2 (Variables|/Script/Engine.Controller|GetPlayerState (Widget|GetOwningPlayer _returnvalue)))))))
            (:then_1
              (if (Variables|Default|GetIsMyPortal)
                (Utilities|IsValid (Variables|Default|GetBeacon)
                  (:Is%20Not%20Valid
                    (bind _returnvalue_3 (Game|SpawnActorfromClass "/Game/Playgrounds/Items/Portals/BP_PortalLightPillar.BP_PortalLightPillar_C" (Transformation|GetActorTransform) "Undefined" "MultiplyWithRoot" 0 0))
                    (Variables|Default|SetBeacon _returnvalue_3)))
                (else
                  (Actor|DestroyActor (Variables|Default|GetBeacon)))))
            (:then_2
              (if (Variables|Portal|GetbPortalOpen)
                (if (Variables|Default|GetIsMyPortal)
                  (Rendering|Material|SetMaterial _sm_rift_2 0 "/Game/Athena/Items/ForagedItems/Rift/FX/M_Shard_Portal_Gold_Inst.M_Shard_Portal_Gold_Inst")
                  (Components|Activation|Deactivate _p_riftclosingup)
                  (Components|Activation|Deactivate _ns_rift_idle_loop)
                  (Components|Activation|Activate _p_rift_idle_loop_gold false)
                  (else
                    (Rendering|Material|SetMaterial _sm_rift_2 0 "/Game/Athena/Items/ForagedItems/Rift/FX/M_Shard_Portal.M_Shard_Portal")
                    (Components|Activation|Deactivate _p_riftclosingup)
                    (Components|Activation|Activate _ns_rift_idle_loop false)
                    (Components|Activation|Deactivate _p_rift_idle_loop_gold)))
                (else
                  (Components|Activation|Activate (Variables|VFX|GetP_RiftClosingUp) false)
                  (Components|Activation|Deactivate (Variables|Default|GetP_Rift_Idle_Loop_Gold)))))
            (:then_3
              (if (Variables|Portal|GetbPortalOpen)
                (bind (_direction _open) (Timeline|Timeline_0 Play)
                  (:Update
                    (Transformation|SetRelativeScale3D _sm_rift (Math|Conversions|ToVector(Float) _open))
                    (Rendering|Material|SetScalarParameterValueonMaterials _sm_rift "CloseMult" _open)
                    (Transformation|SetRelativeRotation _sm_rift (Math|Rotator|MakeRotator (* _open -2e+01)) false false)))
                (else
                  (Timeline|Timeline_0 Reverse))))
            (:then_4
              (CallFunction|UpdatePopulationDisplay)
              (CallFunction|UpdateNameDisplay)
              (CallFunction|CheckShouldDisplayUI))))))
    (:then_1
      (CodeOverride|UpdateBaseVisibility)
      (CodeOverride|UpdateCodeEntryVisibility)
      (CodeOverride|UpdateCodeEntryScreen true))))

(event OnComponentBeginOverlap(HUDVisibilityVolume)
  (params (OverlappedComponent :type Primitive%20Component%20Object%20Reference) (OtherActor :type Actor%20Object%20Reference) (OtherComp :type Primitive%20Component%20Object%20Reference) (OtherBodyIndex :type Integer) (bFromSweep :type Boolean) (SweepResult :type Hit%20Result%20Structure%20%28by%20ref%29))
  ;[1 1 1] Start Detecting Looking at the Code Screen or not
  (CodeOverride|LocalPlayerRangeCheck OtherActor true))

(event Custom|EvaluateOverlap
  (bind _sm_rift (Variables|Default|GetSM_Rift))
  ;[1 1 1] On Overlap - Toggle Post Process
  (if (> (Utilities|Array|Length (Teleport|GetOverlappingPlayerPawns self)) 0)
    ;;[1 1 1] Widen rift if player is close
    (bind (_direction _open) (Timeline|VectorScale Play)
      (:Update
        (Rendering|Material|SetScalarParameterValueonMaterials _sm_rift "Open" (Math|Float|SelectFloat _open 0.0 (Variables|Portal|GetbPortalOpen)))
        (Transformation|SetRelativeRotation _sm_rift (Math|Rotator|MakeRotator (Math|Float|SelectFloat (* _open -2e+01) 0.0 (Variables|Portal|GetbPortalOpen))) false false)
        (Effects|Components|ParticleSystem|SetFloatParameter (Variables|Default|GetNS_Rift_Idle_Loop) "Open" (Math|Float|SelectFloat _open 0.0 (Variables|Portal|GetbPortalOpen)))))
    (else
      (Timeline|VectorScale Reverse))))

(event OnComponentEndOverlap(HUDVisibilityVolume)
  (params (OverlappedComponent :type Primitive%20Component%20Object%20Reference) (OtherActor :type Actor%20Object%20Reference) (OtherComp :type Primitive%20Component%20Object%20Reference) (OtherBodyIndex :type Integer))
  ;[1 1 1] Start Detecting Looking at the Code Screen or not
  (CodeOverride|LocalPlayerRangeCheck OtherActor))

(event OnComponentBeginOverlap(Capsule)
  (params (OverlappedComponent :type Primitive%20Component%20Object%20Reference) (OtherActor :type Actor%20Object%20Reference) (OtherComp :type Primitive%20Component%20Object%20Reference) (OtherBodyIndex :type Integer) (bFromSweep :type Boolean) (SweepResult :type Hit%20Result%20Structure%20%28by%20ref%29))
  (bind _sm_rift (Variables|Default|GetSM_Rift))
  ;[1 1 1] On Overlap - Toggle Post Process
  (Utilities|IsValid (Teleport|GetPlayerPawnfromActor OtherActor)
    (:Is%20Valid
      (CallFunction|PlayExpandorCollapseAudio true)
      ;;[1 1 1] Widen rift if player is close
      (Timeline|VectorScale Play))))

(event EventServerOnInteractWithIslandPortal
  (params (InteractingPawn :type Fort%20Player%20Pawn%20Object%20Reference))
  (Teleport|StartTeleportbyInteractiononServer InteractingPawn))

(event OnComponentEndOverlap(Capsule)
  (params (OverlappedComponent :type Primitive%20Component%20Object%20Reference) (OtherActor :type Actor%20Object%20Reference) (OtherComp :type Primitive%20Component%20Object%20Reference) (OtherBodyIndex :type Integer))
  (bind _sm_rift (Variables|Default|GetSM_Rift))
  ;[1 1 1] On Overlap - Toggle Post Process
  (if (<= (Utilities|Array|Length (Teleport|GetOverlappingPlayerPawns OtherActor)) 0)
    (if (HelperFunctions|IsLocalPawn OtherActor)
      (CallFunction|LogIslandTeleportation))
    (CallFunction|PlayExpandorCollapseAudio false)
    ;;[1 1 1] Widen rift if player is close
    (Timeline|VectorScale Reverse)))

(event EventCuratedPortalChanged)

(event EventClientSaveDisplayChanged
  (params (bSaving :type Boolean))
  (CallFunction|UpdateNameDisplay))

(event EventNotifyTeleportedPlayerPawn
  (params (PlayerPawn :type Fort%20Player%20Pawn%20Object%20Reference) (bTeleportedToIslandStart :type Boolean))
  (bind _self self)
  (bind _returnvalue (Minigame|GetMinigame _self :!class /Script/FortniteGame.MinigameBlueprintLibrary))
  (bind _delayedskydiveplayers (Variables|RiftPortal|GetDelayedSkydivePlayers))
  (CallFunction|PlayTeleportTransition (Transformation|GetActorLocation PlayerPawn) true)
  (if (and (Variables|RiftPortal|GetStartSkydiving) (not bTeleportedToIslandStart))
    (if (CallFunction|ShouldDelaySkydive PlayerPawn)
      (Utilities|IsValid _returnvalue
        (:Is%20Valid
          (Utilities|Array|AddUnique _delayedskydiveplayers PlayerPawn)
          (if (== (Utilities|Array|Length _delayedskydiveplayers) 1)
            (Minigame|BindEventtoPlayerAddedPostResetDelegate _returnvalue (EventDispatchers|CreateEvent :!event Custom|DelayedSkydive))))
        (:Is%20Not%20Valid
          (CallFunction|StartSkydiveforPawn PlayerPawn)))
      (else
        (CallFunction|StartSkydiveforPawn PlayerPawn)))
    (else
      (CallFunction|EventMulticastRiftSound (Variables|Default|GetRiftSound) PlayerPawn))))

(event EventDisplayNameChanged
  (CallFunction|UpdateNameDisplay)
  (CallFunction|UpdatePortal))

(event Custom|EventMulticastRiftSound
  (params (SoundToPlay :type Sound%20Base%20Object%20Reference) (Player%20Pawn :type Fort%20Player%20Pawn%20Object%20Reference))
  (if (Pawn|IsLocallyControlled Player%20Pawn)
    (Audio|SpawnSound2D SoundToPlay)))

(event EventPopulationChanged
  (CallFunction|UpdatePopulationDisplay))

(event Custom|DelayedSkydive
  (params (UniqueNetId :type Unique%20Net%20Id%20Structure) (bIsLocalPlayer :type Boolean))
  (bind _fort_player_state (Class|CreativeCommonDeviceFunctionLibrary|GetPlayerStateFromNetID UniqueNetId))
  (bind _self self)
  (bind _returnvalue (Minigame|GetMinigame _self :!class /Script/FortniteGame.MinigameBlueprintLibrary))
  (Utilities|IsValid _fort_player_state
    (:Is%20Valid
      (bind _asfort_player_pawn (Utilities|Casting|CastToFortPlayerPawn (Variables|/Script/Engine.PlayerState|GetPawnPrivate _fort_player_state)))
      (CallFunction|StartSkydiveforPawn _asfort_player_pawn)
      (Utilities|Array|RemoveItem (Variables|RiftPortal|GetDelayedSkydivePlayers) _asfort_player_pawn)
      (if (not (Utilities|IsArrayNonEmpty (Variables|RiftPortal|GetDelayedSkydivePlayers)))
        (Utilities|IsValid _returnvalue
          (:Is%20Valid
            (Minigame|UnbindEventfromPlayerAddedPostResetDelegate _returnvalue (EventDispatchers|CreateEvent :!event Custom|DelayedSkydive))))))))

(event EventPortalOpenChanged
  (Networking|CanExecuteCosmeticEvents
    (:True
      (CallFunction|PlayExpandorCollapseAudio (Variables|Portal|GetbPortalOpen)))
    (:False
      _))
  (CallFunction|UpdatePortal))

(event EventErrorStateChanged
  (CallFunction|UpdatePortal))

(event Custom|ClientTeleportToLobbyIsland
  (params (RequestingController :type Fort%20Player%20Controller%20Object%20Reference))
  (CallFunction|TeleporttoLobbyIsland RequestingController))

(event EventOnReceivedNewIslandInfo
  ;[1 1 1] On Recieved New Island Info
  (Development|SetHiddeninGame (Variables|Default|GetSM_Rift) false false)
  (Niagara|SetNiagaraSystemAsset (Variables|Default|GetNS_Rift_Idle_Loop) "/Game/Athena/Items/ForagedItems/Rift/FX/NS_Rift_Portal_Idle_Shimmer.NS_Rift_Portal_Idle_Shimmer" true))

(event EventOnThumbnailTextureReady
  (bind _thumbnailplane (Variables|Default|GetThumbnailPlane))
  ;[1 1 1] Thumbnails
  (Utilities|IsValid (Variables|FortAthenaCreativePortal|GetThumbnailTexture)
    (:Is%20Valid
      (Development|SetHiddeninGame (Variables|Default|GetThumbnailPlane) false false)
      (bind _returnvalue (Rendering|Material|CreateDynamicMaterialInstance (Variables|Default|GetThumbnailPlane) 0 0 "None" :!class /Script/Engine.PrimitiveComponent))
      (Rendering|Material|SetTextureParameterValue _returnvalue "Texture" (Variables|FortAthenaCreativePortal|GetThumbnailTexture))
      (Rendering|Material|SetVectorParameterValue _returnvalue "Color" (select (Variables|Default|GetIsMyPortal) (Variables|Default|GetOwnedThumbnailColor) (Variables|Default|GetUnownedThumbnailColor)) :!class /Script/Engine.MaterialInstanceDynamic)
      (Transformation|SetAbsolute _thumbnailplane false true true)
      (Transformation|SetWorldRotation _thumbnailplane "0,-90.000000, 0" false false)
      (Transformation|SetWorldScale3D (Variables|Default|GetThumbnailPlane) (CallFunction|ScaleThumbnailPlanebyTextureSize))
      (CallFunction|SetupPortalforThumbnailDisplay))
    (:Is%20Not%20Valid
      (Development|SetHiddeninGame (Variables|Default|GetThumbnailPlane) true false))))

(event Custom|SetupPortalForThumbnailDisplay
  ;[0 0 0] the configuration of the portal when we have a thumbnail available
  (Transformation|SetRelativeLocation (Variables|Default|GetThumbnailPlane) "0, 0,114.000000" false false)
  (Development|SetHiddeninGame (Variables|Default|GetSM_Rift) true false)
  (Niagara|SetNiagaraSystemAsset (Variables|Default|GetNS_Rift_Idle_Loop) "/Game/Athena/Items/ForagedItems/Rift/FX/NS_Rift_Portal_Idle.NS_Rift_Portal_Idle" true))

(fn BlueprintCanInteract
  (params (InteractingPawn :type Fort%20Pawn%20Object%20Reference) (InteractionBeingAttempted :type EInteractionBeingAttempted%20Enum :default "FirstInteraction") (InteractionType :type TInteractionType%20Enum :default "IT_NoInteraction"))
  ;[0 0.25 0.25] First interaction is enough
  (if (Utilities|Enum|Equal(Enum) InteractionBeingAttempted "FirstInteraction")
    ;[0.25 0 0.25] Show UI info plate if portal is open
    (if (CallFunction|IsPortalOpen)
      (return true)
      (elif (Variables|CodeOverride|GetbLocalPlayerLookingAtScreen)
        ;[0.25 0.25 0] Show UI info plate if there is a code screen to use
        (return (CodeOverride|PortalUsesCodeScreen))
        (else
          (return false))))
    (else
      (return false))))

(fn BlueprintDisplayMobileInteractPrompt
  (return (Interaction|MobileInteractIsValid)))

(fn BlueprintGetInteractionString
  (params (InteractingPawn :type Fort%20Pawn%20Object%20Reference) (InteractionBeingAttempted :type EInteractionBeingAttempted%20Enum :default "FirstInteraction"))
  (return (select (CallFunction|CanModifyPortal) (select (or (Variables|CodeOverride|GetbLocalPlayerLookingAtScreen) (not (CodeOverride|PortalUsesCodeScreen))) (select (Variables|CodeOverride|GetbLocalPlayerLookingAtScreen) (select (TouchPlatform|IsMobileGame) "NSLOCTEXT(\"[BFD1A67D457939B16F3AFE8071909F45]\", \"2EEC214044A87F7FB63BB49ABEDEAE51\", \"USE INTERACT BUTTON TO USE CONSOLE\")" "NSLOCTEXT(\"[BFD1A67D457939B16F3AFE8071909F45]\", \"658E87E34FD76223B2E790B95258B0F5\", \"USE CONSOLE\")") (select (TouchPlatform|IsMobileGame) (select false "NSLOCTEXT(\"[BFD1A67D457939B16F3AFE8071909F45]\", \"13FA076442C4006E2724A5BD3FBB3BE7\", \"USE INTERACT BUTTON TO CHANGE DESTINATION\")" "NSLOCTEXT(\"[BFD1A67D457939B16F3AFE8071909F45]\", \"B1554A9943B8A53B18438590C50264DD\", \"USE INTERACT BUTTON TO CHANGE ISLAND\")") (select false "NSLOCTEXT(\"[BFD1A67D457939B16F3AFE8071909F45]\", \"B848EEA4493BD4EC671365BCB8654B35\", \"CHANGE DESTINATION\")" "NSLOCTEXT(\"[BFD1A67D457939B16F3AFE8071909F45]\", \"A25D08454DD9CD8ACAF6EEA909714CCC\", \"CHANGE ISLAND\")"))) (select (CodeOverride|PortalTeleportsbyInteraction) (select (TouchPlatform|IsMobileGame) "NSLOCTEXT(\"[BFD1A67D457939B16F3AFE8071909F45]\", \"5641F0394D770D92FB3FC7A2BC97075C\", \"USE INTERACT BUTTON TO ENTER PORTAL\")" "NSLOCTEXT(\"[BFD1A67D457939B16F3AFE8071909F45]\", \"B896049A41351F66B538C995A9972A85\", \"ENTER PORTAL\")") "")) "")))

(fn BlueprintGetInteractionTime
  (params (InteractingPawn :type Fort%20Pawn%20Object%20Reference) (InteractionBeingAttempted :type EInteractionBeingAttempted%20Enum :default "FirstInteraction"))
  (if (CodeOverride|PortalUsesCodeScreen)
    (return 0.0 true)
    (elif (CallFunction|CanModifyPortal)
      (return 0.0 false)
      (else
        ;[1 1 1] Stops the character interaction animation
        (return 0.0 true)))))

(fn BlueprintOnLocalInteract
  (params (InteractingPawn :type Fort%20Player%20Pawn%20Object%20Reference) (InteractionBeingAttempted :type EInteractionBeingAttempted%20Enum :default "FirstInteraction"))
  (if (CallFunction|CanModifyPortal)
    (CallFunction|LoadIslandUI InteractingPawn)
    (bind _asfort_player_controller_athena (Utilities|Casting|CastToFortPlayerControllerAthena (Pawn|GetController InteractingPawn))
      (:then
        (if (CallFunction|IsRunningMatchmakingFlowtoEdit _asfort_player_controller_athena)
          (CallFunction|AttemptCancelMatchmakingtoEdit _asfort_player_controller_athena)))
      (:CastFailed
        _))
    (return true)
    (elif (and (CodeOverride|PortalTeleportsbyInteraction) (Variables|CodeOverride|GetbLocalPlayerLookingAtPortal))
      (bind _asathena_player_controller (Utilities|Casting|CastToAthena_PlayerController (Fortnite|GetFortPlayerControllerfromActor InteractingPawn)))
      (Class|FortPlayerControllerAthena|ServerStartInteractwithIslandPortal _asathena_player_controller self)
      (return true)
      (else
        (return true)))))

(fn BlueprintUseMobileHighlight
  (return (Interaction|MobileInteractIsValid)))

(fn CanModifyPortal
  (return (and (or (Variables|CodeOverride|GetbLocalPlayerLookingAtScreen) (and (not (CodeOverride|PortalUsesCodeScreen)) (Variables|CodeOverride|GetbLocalPlayerLookingAtPortal))) (Utilities|CanInteract))))

(fn CanUseHUDInteract
  (return (Utilities|CanInteract)))

(fn CheckForRiftLocUpdate
  (params (InController :type Controller%20Object%20Reference))
  (bind _returnvalue (Pawn|GetControlledPawn InController))
  (bind _riftlocationnum (Variables|RiftPortal|GetRiftLocationNum))
  (bind _returnvalue_2 (Utilities|Select :Index _riftlocationnum :Option%200 "battlepass_athena_use_location_riftportal_01" :Option%201 "battlepass_athena_use_location_riftportal_02" :Option%202 "battlepass_athena_use_location_riftportal_03" :Option%203 "battlepass_athena_use_location_riftportal_04" :Option%204 "battlepass_athena_use_location_riftportal_05" :Option%205 "battlepass_athena_use_location_riftportal_06" :Option%206 "battlepass_athena_use_location_riftportal_07" :Option%207 "battlepass_athena_use_location_riftportal_08" :Option%208 "battlepass_athena_use_location_riftportal_09" :Option%209 " battlepass_athena_use_location_riftportal_10" :Option%2010 "battlepass_athena_use_location_riftportal_11" :Option%2011 "battlepass_athena_use_location_riftportal_12" :Option%2012 "battlepass_athena_use_location_riftportal_13" :Option%2013 "battlepass_athena_use_location_riftportal_14" :Option%2014 "battlepass_athena_use_location_riftportal_15" :Option%2015 "battlepass_athena_use_location_riftportal_16" :Option%2016 "battlepass_athena_use_location_riftportal_17" :Option%2017 "battlepass_athena_use_location_riftportal_18" :Option%2018 "battlepass_athena_use_location_riftportal_19" :Option%2019 "battlepass_athena_use_location_riftportal_20" :Option%2020 "battlepass_athena_use_location_riftportal_21" :Option%2021 "battlepass_athena_use_location_riftportal_22" :Option%2022 "battlepass_athena_use_location_riftportal_23" :Option%2023 "battlepass_athena_use_location_riftportal_24"))
  (bind _riftportallocationsquest (Variables|QuestStuff|GetRiftPortalLocationsQuest))
  (Utilities|IsValid InController
    (:Is%20Valid
      (bind _asfort_player_controller (Utilities|Casting|CastToFortPlayerController InController))
      (bind _returnvalue_3 (GameplayTags|HasMatchingGameplayTag(Message) _returnvalue "(TagName=\"Athena.Location.Meta.RiftPortalLoc.01\")"))
      (if _returnvalue_3
        (Variables|RiftPortal|SetRiftLocationNum 0)
        (else
          (bind _returnvalue_4 (GameplayTags|HasMatchingGameplayTag(Message) _returnvalue "(TagName=\"Athena.Location.Meta.RiftPortalLoc.02\")"))
          (if _returnvalue_4
            (Variables|RiftPortal|SetRiftLocationNum 1)
            (else
              (bind _returnvalue_5 (GameplayTags|HasMatchingGameplayTag(Message) _returnvalue "(TagName=\"Athena.Location.Meta.RiftPortalLoc.03\")"))
              (if _returnvalue_5
                (Variables|RiftPortal|SetRiftLocationNum 2)
                (else
                  (bind _returnvalue_6 (GameplayTags|HasMatchingGameplayTag(Message) _returnvalue "(TagName=\"Athena.Location.Meta.RiftPortalLoc.04\")"))
                  (if _returnvalue_6
                    (Variables|RiftPortal|SetRiftLocationNum 3)
                    (else
                      (bind _returnvalue_7 (GameplayTags|HasMatchingGameplayTag(Message) _returnvalue "(TagName=\"Athena.Location.Meta.RiftPortalLoc.05\")"))
                      (if _returnvalue_7
                        (Variables|RiftPortal|SetRiftLocationNum 4)
                        (else
                          (bind _returnvalue_8 (GameplayTags|HasMatchingGameplayTag(Message) _returnvalue "(TagName=\"Athena.Location.Meta.RiftPortalLoc.06\")"))
                          (if _returnvalue_8
                            (Variables|RiftPortal|SetRiftLocationNum 5)
                            (else
                              (bind _returnvalue_9 (GameplayTags|HasMatchingGameplayTag(Message) _returnvalue "(TagName=\"Athena.Location.Meta.RiftPortalLoc.07\")"))
                              (if _returnvalue_9
                                (Variables|RiftPortal|SetRiftLocationNum 6)
                                (else
                                  (bind _returnvalue_10 (GameplayTags|HasMatchingGameplayTag(Message) _returnvalue "(TagName=\"Athena.Location.Meta.RiftPortalLoc.08\")"))
                                  (if _returnvalue_10
                                    (Variables|RiftPortal|SetRiftLocationNum 7)
                                    (else
                                      (bind _returnvalue_11 (GameplayTags|HasMatchingGameplayTag(Message) _returnvalue "(TagName=\"Athena.Location.Meta.RiftPortalLoc.09\")"))
                                      (if _returnvalue_11
                                        (Variables|RiftPortal|SetRiftLocationNum 8)
                                        (else
                                          (bind _returnvalue_12 (GameplayTags|HasMatchingGameplayTag(Message) _returnvalue "(TagName=\"Athena.Location.Meta.RiftPortalLoc.10\")"))
                                          (if _returnvalue_12
                                            (Variables|RiftPortal|SetRiftLocationNum 9)
                                            (else
                                              (bind _returnvalue_13 (GameplayTags|HasMatchingGameplayTag(Message) _returnvalue "(TagName=\"Athena.Location.Meta.RiftPortalLoc.11\")"))
                                              (if _returnvalue_13
                                                (Variables|RiftPortal|SetRiftLocationNum 10)
                                                (else
                                                  (bind _returnvalue_14 (GameplayTags|HasMatchingGameplayTag(Message) _returnvalue "(TagName=\"Athena.Location.Meta.RiftPortalLoc.12\")"))
                                                  (if _returnvalue_14
                                                    (Variables|RiftPortal|SetRiftLocationNum 11)
                                                    (else
                                                      (bind _returnvalue_15 (GameplayTags|HasMatchingGameplayTag(Message) _returnvalue "(TagName=\"Athena.Location.Meta.RiftPortalLoc.13\")"))
                                                      (if _returnvalue_15
                                                        (Variables|RiftPortal|SetRiftLocationNum 12)
                                                        (else
                                                          (bind _returnvalue_16 (GameplayTags|HasMatchingGameplayTag(Message) _returnvalue "(TagName=\"Athena.Location.Meta.RiftPortalLoc.14\")"))
                                                          (if _returnvalue_16
                                                            (Variables|RiftPortal|SetRiftLocationNum 13)
                                                            (else
                                                              (bind _returnvalue_17 (GameplayTags|HasMatchingGameplayTag(Message) _returnvalue "(TagName=\"Athena.Location.Meta.RiftPortalLoc.15\")"))
                                                              (if _returnvalue_17
                                                                (Variables|RiftPortal|SetRiftLocationNum 14)
                                                                (else
                                                                  (bind _returnvalue_18 (GameplayTags|HasMatchingGameplayTag(Message) _returnvalue "(TagName=\"Athena.Location.Meta.RiftPortalLoc.16\")"))
                                                                  (if _returnvalue_18
                                                                    (Variables|RiftPortal|SetRiftLocationNum 15)
                                                                    (else
                                                                      (bind _returnvalue_19 (GameplayTags|HasMatchingGameplayTag(Message) _returnvalue "(TagName=\"Athena.Location.Meta.RiftPortalLoc.17\")"))
                                                                      (if _returnvalue_19
                                                                        (Variables|RiftPortal|SetRiftLocationNum 16)
                                                                        (else
                                                                          (bind _returnvalue_20 (GameplayTags|HasMatchingGameplayTag(Message) _returnvalue "(TagName=\"Athena.Location.Meta.RiftPortalLoc.18\")"))
                                                                          (if _returnvalue_20
                                                                            (Variables|RiftPortal|SetRiftLocationNum 17)
                                                                            (else
                                                                              (bind _returnvalue_21 (GameplayTags|HasMatchingGameplayTag(Message) _returnvalue "(TagName=\"Athena.Location.Meta.RiftPortalLoc.19\")"))
                                                                              (if _returnvalue_21
                                                                                (Variables|RiftPortal|SetRiftLocationNum 18)
                                                                                (else
                                                                                  (bind _returnvalue_22 (GameplayTags|HasMatchingGameplayTag(Message) _returnvalue "(TagName=\"Athena.Location.Meta.RiftPortalLoc.20\")"))
                                                                                  (if _returnvalue_22
                                                                                    (Variables|RiftPortal|SetRiftLocationNum 19)
                                                                                    (else
                                                                                      (bind _returnvalue_23 (GameplayTags|HasMatchingGameplayTag(Message) _returnvalue "(TagName=\"Athena.Location.Meta.RiftPortalLoc.21\")"))
                                                                                      (if _returnvalue_23
                                                                                        (Variables|RiftPortal|SetRiftLocationNum 20)
                                                                                        (else
                                                                                          (bind _returnvalue_24 (GameplayTags|HasMatchingGameplayTag(Message) _returnvalue "(TagName=\"Athena.Location.Meta.RiftPortalLoc.22\")"))
                                                                                          (if _returnvalue_24
                                                                                            (Variables|RiftPortal|SetRiftLocationNum 21)
                                                                                            (else
                                                                                              (bind _returnvalue_25 (GameplayTags|HasMatchingGameplayTag(Message) _returnvalue "(TagName=\"Athena.Location.Meta.RiftPortalLoc.23\")"))
                                                                                              (if _returnvalue_25
                                                                                                (Variables|RiftPortal|SetRiftLocationNum 22)
                                                                                                (else
                                                                                                  (bind _returnvalue_26 (GameplayTags|HasMatchingGameplayTag(Message) _returnvalue "(TagName=\"Athena.Location.Meta.RiftPortalLoc.24\")"))
                                                                                                  (if _returnvalue_26
                                                                                                    (Variables|RiftPortal|SetRiftLocationNum 23)
                                                                                                    (else
                                                                                                      (Development|PrintString "No  Rift Loc Tags Present" true true "(R=0.647162,G=0.000000,B=1.000000,A=1.000000)" 5.0)))))))))))))))))))))))))))))))))))))))))))))))))
      (Development|PrintString (Utilities|String|ToString(Integer) _riftlocationnum))
      (bind (_questisvalid _questcompleted) (Default|AthenaCheckQuestandObjectiveCompletion _asfort_player_controller _riftportallocationsquest _returnvalue_2))
      (if (and _questisvalid (not _questcompleted))
        (Class|AthenaQuestFunctionLibrary|AthenaGiveQuestUpdatetoPlayers :InQuestItemRef _riftportallocationsquest :ObjectiveBackendName _returnvalue_2 :PlayerControllersForUpdate (Utilities|Array|MakeArray _asfort_player_controller))
        (Development|PrintString "SUCCESS!" true true "(R=0.416182,G=1.000000,B=0.000000,A=1.000000)" 1e+01)
        (else
          (Development|PrintString "Quest Not Valid" true true "(R=1.000000,G=0.066515,B=0.000000,A=1.000000)" 1e+01))))))

(fn ClearDisplayedWidget
  (Variables|UI|SetDisplayedWidget 0))

(fn CloseAndClearDisplayedWidget
  (Utilities|FlowControl|Sequence
    (:then_0
      (Utilities|IsValid (Variables|UI|GetDisplayedWidget)
        (:Is%20Valid
          (Default|CallCloseShownUI))))
    (:then_1
      (Variables|UI|SetDisplayedWidget 0))))

(fn DetermineWherePlayerIsLooking
  (params (Pawn :type Pawn%20Object%20Reference))
  (bind _returnvalue (Pawn|GetController Pawn))
  (bind (_outlocation _outrotation) (Actor|GetActorEyesViewPoint _returnvalue))
  (Utilities|IsValid Pawn
    (:Is%20Valid
      (Utilities|IsValid _returnvalue
        (:Is%20Valid
          (LocalVariables|SetEyesViewPointLocation _outlocation)
          (LocalVariables|SetEyesViewPointRotation _outrotation)
          (LocalVariables|SetbLookingAtPortal (CallFunction|IsLookingatSphere (Variables|Default|GetPortalLookAtSphere) (LocalVariables|GetEyesViewPointLocation) (LocalVariables|GetEyesViewPointRotation)))
          (if (CodeOverride|PortalUsesCodeScreen)
            (LocalVariables|SetbLookingAtScreen (CallFunction|IsLookingatSphere (Variables|Default|GetCodeScreenLookAtSphere) (LocalVariables|GetEyesViewPointLocation) (LocalVariables|GetEyesViewPointRotation)))
            (if (and (LocalVariables|GetbLookingAtScreen) (LocalVariables|GetbLookingAtPortal))
              (if (< (Math|Vector|DistanceSquared(Vector) (LocalVariables|GetEyesViewPointLocation) (Transformation|GetWorldLocation (Variables|Default|GetCodeScreenLookAtSphere))) (Math|Vector|DistanceSquared(Vector) (LocalVariables|GetEyesViewPointLocation) (Transformation|GetWorldLocation (Variables|Default|GetPortalLookAtSphere))))
                (LocalVariables|SetbLookingAtPortal false)
                (else
                  (LocalVariables|SetbLookingAtScreen false)))))
          (return (LocalVariables|GetbLookingAtScreen) (LocalVariables|GetbLookingAtPortal)))
        (:Is%20Not%20Valid
          _)))
    (:Is%20Not%20Valid
      _))
  (return false false))

(fn GetIconPlacement
  (params (SelfActor :type Actor%20Object%20Reference) (ViewingActor :type Actor%20Object%20Reference))
  (bind (_outlocation _outextents) (|CallParentFunction 0))
  (if (and (Variables|CodeOverride|GetbLocalPlayerLookingAtScreen) (CodeOverride|PortalUsesCodeScreen))
    ;[1 1 1] Player is looking at the code screen
    (return (Transformation|GetWorldLocation (Variables|Default|GetCodeScreenLookAtSphere)) (Math|Conversions|ToVector(Float) (Variables|/Script/Engine.SphereComponent|GetSphereRadius (Variables|Default|GetCodeScreenLookAtSphere))))
    (else
      ;[1 1 1] Player is looking at the portal
      (return _outlocation _outextents))))

(fn GetOverlappingPlayerPawns
  (params (Actor :type Actor%20Object%20Reference))
  (for _array_element (Collision|GetOverlappingActors Actor "/Script/FortniteGame.FortPlayerPawn")
    (bind _asfort_player_pawn (Utilities|Casting|CastToFortPlayerPawn _array_element))
    (Utilities|Array|AddUnique (LocalVariables|GetPlayerPawns) _asfort_player_pawn))
  (for _array_element_2 (Collision|GetOverlappingActors Actor "/Script/FortniteGame.FortAthenaVehicle")
    (Utilities|IsValid (Teleport|GetPlayerPawnfromActor _array_element_2)
      (:Is%20Valid
        (Utilities|Array|AddUnique (LocalVariables|GetPlayerPawns) (Teleport|GetPlayerPawnfromActor _array_element_2)))))
  (return (LocalVariables|GetPlayerPawns)))

(fn GetPlayerPawnFromActor
  (params (Actor :type Actor%20Object%20Reference))
  (bind _asfort_player_pawn (Utilities|Casting|CastToFortPlayerPawn Actor)
    (:then
      (return _asfort_player_pawn))
    (:CastFailed
      (bind _asfort_athena_vehicle (Utilities|Casting|CastToFortAthenaVehicle Actor)
        (:then
          (return (Vehicle|GetDriver _asfort_athena_vehicle)))
        (:CastFailed
          (return 0))))))

(fn GetPortalPosition
  (bind _portalcount (Variables|RiftPortal|GetPortalCount))
  (Math|Integer|IncrementInt _portalcount
    (:%20%20
      (return (Math|Vector|MakeVector (+ (* _portalcount (ScalableFloat|GetValueatLevel (Variables|RiftPortal|GetXPortalSeparation))) (.x (Transformation|GetActorLocation))) (+ (.y (Transformation|GetActorLocation)) (* _portalcount (ScalableFloat|GetValueatLevel (Variables|RiftPortal|GetYPortalSeparation)))) (.z (Transformation|GetActorLocation)))))))

(fn IconPlacementNeedsUpdate
  (return true))

(fn IsLookingAtSphere
  (params (LookAtSphere :type Sphere%20Collision%20Object%20Reference) (Eyes%20View%20Point%20Location :type Vector) (Eyes%20View%20Point%20Rotation :type Rotator))
  (LocalVariables|SetLookAtWorldLocation (Transformation|GetWorldLocation LookAtSphere))
  (return (< (Math|Vector|Distance2DSquared(Vector) (LocalVariables|GetLookAtWorldLocation) (Math|Vector|FindClosestPointonLine (LocalVariables|GetLookAtWorldLocation) (Variables|GetEyes%20View%20Point%20Location) (Math|Vector|GetForwardVector (Variables|GetEyes%20View%20Point%20Rotation)))) (Math|Float|Square (Components|Sphere|GetScaledSphereRadius (Variables|GetLookAtSphere))))))

(fn LocalPlayerEnteredRange
  (params (LocalPawn :type Pawn%20Object%20Reference))
  (Utilities|Array|AddUnique (Variables|CodeOverride|GetLocalPawnsInRange) LocalPawn)
  (CodeOverride|StartorEndQuery))

(fn LocalPlayerExitedRange
  (params (LocalPawn :type Pawn%20Object%20Reference))
  (Utilities|Array|RemoveItem (Variables|CodeOverride|GetLocalPawnsInRange) LocalPawn)
  (CodeOverride|StartorEndQuery))

(fn LocalPlayerRangeCheck
  (params (OverlappingActor :type Actor%20Object%20Reference) (bEnteredRange :type Boolean))
  (Networking|CanExecuteCosmeticEvents
    (:True
      (bind _output_get (LocalVariables|SetOverlappingPlayerController (Fortnite|GetFortPlayerControllerfromActor OverlappingActor)))
      (Utilities|IsValid _output_get
        (:Is%20Valid
          (bind _aspawn (Utilities|Casting|CastToPawn (Variables|GetOverlappingActor)))
          (if (Utilities|Array|ContainsItem (Fortnite|GetLocalFortPlayerControllers true) (LocalVariables|GetOverlappingPlayerController))
            (if (Variables|GetbEnteredRange)
              (CodeOverride|LocalPlayerEnteredRange _aspawn)
              (else
                (CodeOverride|LocalPlayerExitedRange _aspawn)))))))))

(fn MobileInteractIsValid
  (return (Math|Boolean|ORBoolean (not (CodeOverride|PortalUsesCodeScreen)) (CodeOverride|PortalTeleportsbyInteraction) (and (Variables|CodeOverride|GetbLocalPlayerLookingAtScreen) (CallFunction|CanModifyPortal)))))

(fn OnRep_PlayerPawn)

(fn OnRep_TeleportLocation)

(fn OnRep_Vehicle)

(fn PlayExpandOrCollapseAudio
  (params (Expand :type Boolean))
  (if (Variables|Portal|GetbPortalOpen)
    (Audio|PlaySoundatLocation (select Expand "/Game/Sounds/Fort_GamePlay_Sounds/Player/RiftPortal/RiftPortal_Expand_Cue.RiftPortal_Expand_Cue" "/Game/Sounds/Fort_GamePlay_Sounds/Player/RiftPortal/RiftPortal_Collapse_Cue.RiftPortal_Collapse_Cue") (Transformation|GetWorldLocation (Variables|Default|GetSM_Rift)))))

(fn PortalTeleportsByInteraction
  (return (Math|Boolean|ANDBoolean (Variables|CodeOverride|GetbPlayerMustInteractToUsePortal) (not (Variables|FortAthenaCreativePortal|GetbReturnToCreativeHub)) (Variables|CodeOverride|GetbEnableBaseAndCodeEntryInteractions) (CodeOverride|PortalUsesCodeScreen))))

(fn PortalUsesCodeScreen
  (return (Math|Boolean|ANDBoolean (Variables|CodeOverride|GetbEnableBaseAndCodeEntryInteractions) (Variables|Default|GetbCanPlayerInteract) (Variables|CodeOverride|GetbUseCodeEntryScreen) (or (Online|UniqueNetIdRepl|IsValid(UniqueIdRepl) (Variables|Portal|GetOwningPlayer)) (Variables|Portal|GetbUserInitiatedLoad)))))

(fn QueryPlayerLookAt
  (CodeOverride|UpdateCodeEntryScreen false))

(fn ScaleThumbnailPlaneByTextureSize
  (return (Utilities|Select (Math|Vector|MakeVector (/ (Variables|Default|GetThumbnailWorldSizeScale) (/ (Math|Conversions|ToFloat(Integer) (CallFunction|GetThumbnailWidth)) (Math|Conversions|ToFloat(Integer) (CallFunction|GetThumbnailHeight)))) (Variables|Default|GetThumbnailWorldSizeScale) 1.0) (Math|Vector|MakeVector (Variables|Default|GetThumbnailWorldSizeScale) (/ (Variables|Default|GetThumbnailWorldSizeScale) (/ (Math|Conversions|ToFloat(Integer) (CallFunction|GetThumbnailWidth)) (Math|Conversions|ToFloat(Integer) (CallFunction|GetThumbnailHeight)))) 1.0) (> (CallFunction|GetThumbnailWidth) (CallFunction|GetThumbnailHeight)))))

(fn SetPortalUIVisibility
  (params (Show :type Boolean) (Distance :type Float%20%28double-precision%29))
  (bind _ui_portalinfoplate (Variables|Default|GetUI_PortalInfoPlate))
  (bind _output_get (Variables|UI|SetbIsBeingDisplayed Show))
  (LocalVariables|SetWasHidden (Variables|/Script/Engine.SceneComponent|GetbHiddenInGame _ui_portalinfoplate))
  (Development|SetHiddeninGame _ui_portalinfoplate (not _output_get) false)
  (Default|CallUpdatePortalUIVis (Variables|UI|GetbIsBeingDisplayed) (LocalVariables|GetWasHidden) (Variables|GetDistance)))

(fn ShouldDelaySkydive
  (params (TeleportedPawn :type Fort%20Player%20Pawn%20Object%20Reference))
  (bind _self self)
  (bind _returnvalue (Minigame|GetMinigame _self :!class /Script/FortniteGame.MinigameBlueprintLibrary))
  (Utilities|FlowControl|Sequence
    (:then_0
      (bind _asfort_game_mode_athena (Utilities|Casting|CastToFortGameModeAthena (Game|GetGameMode)))
      (if (Class|FortGameModeAthena|ShouldForceSpawnInsideOfVolume _asfort_game_mode_athena)
        (Utilities|IsValid _returnvalue
          (:Is%20Valid
            (bind _asfort_player_state (Utilities|Casting|CastToFortPlayerState (Variables|/Script/Engine.Pawn|GetPlayerState (Variables|GetTeleportedPawn))))
            (return (Minigame|IsPlayerinOurVolume _returnvalue _asfort_player_state))))))
    (:then_1
      (return false))))

(fn StartOrEndQuery
  (Networking|CanExecuteCosmeticEvents
    (:True
      (if (Utilities|IsArrayNonEmpty (Variables|CodeOverride|GetLocalPawnsInRange))
        (bind _returnvalue (Utilities|Time|SetTimerbyEvent (EventDispatchers|CreateEvent :!event QueryPlayerLookAt) (Variables|CodeOverride|GetLookAtUpdateTimeSecs) true))
        (Variables|CodeOverride|SetLookAtCodeScreenHandle _returnvalue)
        (else
          (Utilities|Time|ClearandInvalidateTimerbyHandle (Variables|CodeOverride|GetLookAtCodeScreenHandle))
          (CodeOverride|UpdateCodeEntryScreen false))))))

(fn StartSkydiveForPawn
  (params (TeleportedPawn :type Fort%20Player%20Pawn%20Object%20Reference))
  (Parachute|BeginSkydiving TeleportedPawn false))

(fn StartTeleport
  (params (OverlapActor :type Actor%20Object%20Reference))
  (bind _overlapactor (Variables|RiftPortal|GetOverlapActor))
  (bind _playerpawn (Teleport|GetPlayerPawnfromActor _overlapactor))
  (Variables|RiftPortal|SetOverlapActor OverlapActor)
  (if (CallFunction|IsPortalOpen)
    (Variables|RiftPortal|SetActorRotation (Transformation|GetActorRotation (Variables|RiftPortal|GetOverlapActor)))
    (if (Variables|FortAthenaCreativePortal|GetbReturnToCreativeHub)
      (CallFunction|TeleporttoCreativeHub)
      (elif (Networking|HasAuthority)
        (Utilities|IsValid _playerpawn
          (:Is%20Valid
            (if (not (Variables|/Script/FortniteGame.FortPawn|GetbIsDBNO _playerpawn))
              (bind _output_get (Variables|RiftPortal|SetPlayerPawn _playerpawn))
              (bind _asfort_player_controller_athena (Utilities|Casting|CastToFortPlayerControllerAthena (Pawn|GetController _output_get)))
              (if (not (Variables|/Script/FortniteGame.FortPlayerControllerAthena|GetbDelayedTeleporting _asfort_player_controller_athena))
                (UI|SetPortalUIVisibility false)
                (CallFunction|PlayTeleportAudio)
                (CallFunction|PlayTeleportFX (Transformation|GetActorLocation _playerpawn) (Variables|Portal|GetTeleportLocation))
                (if (not (Vehicle|IsinVehicle _playerpawn))
                  (Networking|SwitchHasAuthority
                    (:Authority
                      (RemoteControlPawn|ForceDetachfromRemoteControlledPawn _playerpawn)
                      (Variables|/Script/FortniteGame.FortPawn|SetbIgnoreNextFallingDamage true _playerpawn) ; Have to do this because of the delay. Otherwise player can die from fall damage if they fall from high distance.
                      (Utilities|FlowControl|Sequence
                        (:then_0
                          (Parachute|EndSkydiving _playerpawn) ; End Skydiving if player was previously
                          (Utilities|IsValid (CallFunction|GetLinkedVolume)
                            (:Is%20Valid
                              (CallFunction|TeleportPlayertoLinkedVolume _playerpawn true)
                              (Development|PrintText "NSLOCTEXT(\"[BFD1A67D457939B16F3AFE8071909F45]\", \"C02EF8C5484AA47631D07EAD973B9621\", \"Teleport Volume Attempted\")"))
                            (:Is%20Not%20Valid
                              (CallFunction|TeleportPlayer _playerpawn (Variables|RiftPortal|GetActorRotation))
                              (Development|PrintText "NSLOCTEXT(\"[BFD1A67D457939B16F3AFE8071909F45]\", \"EED34122438DE0700A8F8E8D5F10C24E\", \"Teleport Player Attempted\")"))))
                        (:then_1
                          (Variables|Portal|SetTeleportLocation (+ (Variables|Portal|GetTeleportLocation) "0,100.000000, 0")))))))))))))))

(fn StartTeleportByInteractionOnServer
  (params (OverlapActor :type Actor%20Object%20Reference))
  (if (Math|Boolean|ANDBoolean (CodeOverride|PortalTeleportsbyInteraction) (Networking|HasAuthority) (Collision|IsOverlappingActor (Variables|Default|GetHUD%20Visibility%20Volume) OverlapActor :!class /Script/Engine.PrimitiveComponent))
    (Teleport|StartTeleport OverlapActor)))

(fn TeleportToLobbyIsland
  (params (RequestingController :type Fort%20Player%20Controller%20Object%20Reference))
  (bind _travelingpawn (LocalVariables|GetTravelingPawn))
  (LocalVariables|SetTravelingPawn (Game|Player|GetPlayerPawn RequestingController))
  (bind _location (CallFunction|GetPortalPosition))
  (bind _output_get (LocalVariables|SetDestination (Math|Vector|MakeVector (.x _location) (.y _location) (ScalableFloat|GetValueatLevel (Variables|RiftPortal|GetStarterIslandTeleportHeight)))))
  (CallFunction|PlayTeleportTransition _output_get)
  (CallFunction|PlayTeleportFX (Transformation|GetActorLocation (LocalVariables|GetTravelingPawn)) (LocalVariables|GetDestination))
  (if (not (Vehicle|IsDrivingVehicle (LocalVariables|GetTravelingPawn)))
    (RemoteControlPawn|ForceDetachfromRemoteControlledPawn (LocalVariables|GetTravelingPawn))
    (Variables|/Script/FortniteGame.FortPawn|SetbIgnoreNextFallingDamage true _travelingpawn) ; Have to do this because of the delay. Otherwise player can die from fall damage if they fall from high distance.
    (Parachute|EndSkydiving _travelingpawn) ; End Skydiving if player was previously
    (Transform|Actor|TeleportPlayerPawn (LocalVariables|GetTravelingPawn) (LocalVariables|GetDestination) (Transformation|GetActorRotation (LocalVariables|GetTravelingPawn)))
    (Ability|SendGameplayEventtoActor (LocalVariables|GetTravelingPawn) "(TagName=\"Abilities.Generic.Rift.Trigger\")")))

(fn UpdateBaseVisibility
  (LocalVariables|SetbBaseIsHidden (or (not (Variables|CodeOverride|GetbEnableBaseAndCodeEntryInteractions)) (and (not (CodeOverride|PortalUsesCodeScreen)) (not (CallFunction|IsPortalOpen)))))
  (Collision|SetCollisionEnabled (Variables|Default|GetBase%20Mesh) (select (LocalVariables|GetbBaseIsHidden) "NoCollision" "QueryAndPhysics"))
  (Networking|CanExecuteCosmeticEvents
    (:True
      (Development|SetHiddeninGame (Variables|Default|GetBase%20Mesh) (LocalVariables|GetbBaseIsHidden) false))))

(fn UpdateCodeEntryScreen
  (params (bForceScreenUpdate :type Boolean))
  (Networking|CanExecuteCosmeticEvents
    (:True
      (LocalVariables|SetMID (Material|GetMID (Variables|Default|GetCode%20Entry%20Screen)))
      (if (Utilities|IsArrayNonEmpty (Variables|CodeOverride|GetLocalPawnsInRange))
        ;[1 1 1] At Max this will only ever look twice. It's a check to ensure things work for both players in Splitscreen.
        (bind _array_element (Utilities|Array|ForEachLoopwithBreak (Variables|CodeOverride|GetLocalPawnsInRange))
          (:LoopBody
            (bind (_bislookingatscreen _bislookingatportal) (CodeOverride|DetermineWherePlayerIsLooking _array_element))
            (LocalVariables|SetbLookingAtScreen (or _bislookingatscreen (LocalVariables|GetbLookingAtScreen)))
            (LocalVariables|SetbLookingAtPortal (or _bislookingatportal (LocalVariables|GetbLookingAtPortal))))
          (:Completed
            _)))
      (if (Math|Boolean|ORBoolean (Variables|GetbForceScreenUpdate) (Math|Boolean|NotEqual(Boolean) (LocalVariables|GetbLookingAtScreen) (Variables|CodeOverride|GetbLocalPlayerLookingAtScreen)) (Math|Boolean|NotEqual(Boolean) (LocalVariables|GetbLookingAtPortal) (Variables|CodeOverride|GetbLocalPlayerLookingAtPortal)))
        (Variables|CodeOverride|SetbLocalPlayerLookingAtScreen (and (LocalVariables|GetbLookingAtScreen) (not (LocalVariables|GetbLookingAtPortal))))
        (Variables|CodeOverride|SetbLocalPlayerLookingAtPortal (LocalVariables|GetbLookingAtPortal))
        (if (CodeOverride|PortalUsesCodeScreen)
          (Rendering|Material|SetVectorParameterValue (LocalVariables|GetMID) "Screen Tone" (select true (select (Utilities|CanInteract) (select (Variables|CodeOverride|GetbLocalPlayerLookingAtScreen) (Variables|CodeOverride|GetScreenColorCanUpdate) (Variables|CodeOverride|GetScreenColorNotHighlighted)) (Variables|CodeOverride|GetScreenColorLocked)) (Variables|CodeOverride|GetScreenColorDisabled)) :!class /Script/Engine.MaterialInstanceDynamic))
        (UI|UpdateInfoPlatePointer)
        (Default|CallUpdateLocalPlayerLookingatScreen (Variables|CodeOverride|GetbLocalPlayerLookingAtScreen))))))

(fn UpdateCodeEntryVisibility
  (LocalVariables|SetbCodeScreenHidden (or (not (CodeOverride|PortalUsesCodeScreen)) (Variables|FortAthenaCreativePortal|GetbReturnToCreativeHub)))
  (Collision|SetCollisionEnabled (Variables|Default|GetCode%20Entry%20Screen) (select (LocalVariables|GetbCodeScreenHidden) "NoCollision" "QueryAndPhysics"))
  (Networking|CanExecuteCosmeticEvents
    (:True
      (Development|SetHiddeninGame (Variables|Default|GetCode%20Entry%20Screen) (LocalVariables|GetbCodeScreenHidden) false))))

(fn UpdateInfoPlatePointer
  (CallFunction|SetHUDPointerOrigin (Transformation|GetWorldLocation (select (Math|Boolean|ANDBoolean (or (Variables|CodeOverride|GetbLocalPlayerLookingAtScreen) (not (CallFunction|IsPortalOpen))) (CodeOverride|PortalUsesCodeScreen) (Utilities|CanInteract)) (Variables|Default|GetCodeScreenLookAtSphere) (Variables|Default|GetPortalHUDDetails_PointerOrigin)))))

(fn UpdateNameDisplay
  (bind _returnvalue (CallFunction|GetMnemonic))
  (Default|CallUpdateIslandName (CallFunction|GetVolumeDisplayName))
  (Default|CallUpdateIsCuratedPortal (CallFunction|GetIsPublishedPortal) (> (Utilities|String|Len _returnvalue) 0) (Variables|Portal|GetbUserInitiatedLoad))
  (Default|CallUpdateLinkCode _returnvalue)
  (if (Utilities|Text|TextIsEmpty (Variables|Portal|GetOverrideName))
    (if (Variables|FortAthenaCreativePortal|GetbReturnToCreativeHub)
      (Default|CallUpdatePlayerName "NSLOCTEXT(\"[BFD1A67D457939B16F3AFE8071909F45]\", \"AD5F33134C164BE9871D4BBB4A96C8E9\", \"CREATIVE HUB\")")
      (Default|CallUpdateDescription "")
      (else
        (bind _returnvalue_2 (CallFunction|GetPlayerState))
        (if (Utilities|IsValid _returnvalue_2 :!class /Script/Engine.KismetSystemLibrary)
          (Default|CallUpdatePlayerName (Utilities|Text|ToText(String) (PlayerState|GetPlayerNameforStreaming _returnvalue_2)))
          (else
            (Default|CallUpdatePlayerName "")))
        (Default|CallUpdateDescription (CallFunction|GetTagline))))
    (else
      (Default|CallUpdatePlayerName (Utilities|Text|TexttoUpper (Variables|Portal|GetOverrideName)))
      (Default|CallUpdateDescription ""))))

(fn UpdatePopulationDisplay
  (bind _maxavailablepopulation (Variables|FortAthenaCreativePortal|GetMaxAvailablePopulation))
  (bind _returnvalue (Math|Conversions|ToInteger(Byte) _maxavailablepopulation))
  (if (Variables|Portal|GetbPortalOpen)
    (Default|CallUpdatePlayerPop (Math|Conversions|ToInteger(Byte) (Variables|FortAthenaCreativePortal|GetCurrentPopulation)) _returnvalue)
    (else
      (Default|CallUpdatePlayerPop 0 _returnvalue))))

(fn ConstructionScript)

(fn UseRiftQuestUpdate
  (params (InController :type Controller%20Object%20Reference))
  (Utilities|IsValid InController
    (:Is%20Valid
      (bind _asfort_player_controller (Utilities|Casting|CastToFortPlayerController InController))
      (bind (_questisvalid _questcompleted) (Default|AthenaCheckQuestandObjectiveCompletion _asfort_player_controller (Variables|QuestStuff|GetQuestItem) (Variables|QuestStuff|GetObjectiveBackendName)))
      (if (and _questisvalid (not _questcompleted))
        (Class|AthenaQuestFunctionLibrary|AthenaGiveQuestUpdatetoPlayers :InQuestItemRef (Variables|QuestStuff|GetQuestItem) :ObjectiveBackendName (Variables|QuestStuff|GetObjectiveBackendName) :PlayerControllersForUpdate (Utilities|Array|MakeArray _asfort_player_controller))))))
