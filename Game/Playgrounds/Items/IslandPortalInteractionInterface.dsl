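; Blueprint interface for island portal items that exposes an interaction notification to implementers.
; It declares one function, BroadcastIsBeingInteractedWith, which takes a Boolean IsInteracting flag and returns nothing.
; The file holds only the signature, with no logic.
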(fn BroadcastIsBeingInteractedWith
  (params (IsInteracting :type Boolean)))
