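# Translucent two-sided island portal material, showing a destination image parameter over a scrolling nebula, framed by glowing energy rings.
# Two panning smoke samples distort UV1 for the plasma strands, normal-driven ring gradient, nebula and noisy outer edge.
# Sphere masks separate and blend these layers, with BumpOffset giving image and nebula false depth.
# Time-scrolled curl noise in world space, masked along UV1, produces a wobble offset.
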Material /Game/Playgrounds/Items/M_IslandPortal
graph {
 properties {
  bEnableResponsiveAA=true
  blendMode=BLEND_Translucent
  pixelDepthOffsetMode=PDOM_Legacy
  translucencyLightingMode=TLM_VolumetricPerVertexDirectional
  twoSided=true
  usedShadingModels=Default Lit
 }
 nodes [
  n1 = TextureCoordinate() {coordinateIndex=1}
  n2 = VectorParameter() {parameterName=TilePanA, defaultValue=<rgba 1.344 0.8064 0.05 0.0433>}
  n3 = AppendVector(A=@n2.R, B=@n2.R)
  n4 = TextureSample(UVs=Panner(Coordinate=Multiply(A=@n1, B=@n3), Speed=AppendVector(A=@n2.B, B=@n2.B)) {speedX=0.1, speedY=0.1}) {texture=&T_Inky_Smoke_Tile}
  n5 = TextureCoordinate() {coordinateIndex=1}
  n6 = VectorParameter() {parameterName=TilePanB, defaultValue=<rgba 1.68 1.008 -0.04 -0.0532>}
  n7 = AppendVector(A=@n6.R, B=@n6.R)
  n8 = TextureSample(UVs=Panner(Coordinate=Multiply(A=@n5, B=@n7), Speed=AppendVector(A=@n6.B, B=@n6.B)) {speedX=0.1, speedY=0.1}) {texture=&T_Inky_Smoke_Tile}
  n9 = Multiply(A=Subtract(A=AppendVector(A=@n4.R, B=@n8.R)) {constB=0.5}, B=ScalarParameter() {defaultValue=0.05, parameterName=UVNoiseStrength})
  n10 = Multiply(
   A=Add(A=TextureCoordinate() {coordinateIndex=1}, B=@n9)
   B=AppendVector(A=ScalarParameter() {defaultValue=2, parameterName=Ub}, B=ScalarParameter() {defaultValue=0.24, parameterName=Vb})
  )
  n11 = AppendVector(A=ScalarParameter() {parameterName=PanSpeedUb}, B=ScalarParameter() {defaultValue=0.1, parameterName=PanSpeedVb})
  n12 = SphereMask(
   A=ComponentMask(None=TextureCoordinate() {coordinateIndex=1}) {g=true}
   B=ScalarParameter() {defaultValue=0.75, parameterName=CelestialCenter}
   Radius=ScalarParameter() {defaultValue=0.5, parameterName=CelestialRadius}
   Hardness=ScalarParameter() {defaultValue=0.2, parameterName=CelestialHardness}
  )
  n13 = Multiply(
   A=TextureSample(UVs=Panner(Coordinate=@n10, Speed=@n11)) {texture=/Game/Effects/Fort_Effects/Textures/Plasma/T_Plasma_Aurora_Seamless}.B
   B=@n12
  )
  n14 = Add(A=@n9, B=OneMinus(None=TextureCoordinate() {coordinateIndex=1}))
  n15 = ScalarParameter() {defaultValue=0.5, parameterName=Bias}
  n16 = Add(A=@n9, B=TextureCoordinate() {coordinateIndex=1})
  n17 = AppendVector(A=ScalarParameter() {defaultValue=2, parameterName=Ua}, B=ScalarParameter() {defaultValue=1, parameterName=Va})
  n18 = TextureSample(UVs=Panner(Coordinate=Multiply(A=@n16, B=@n17)) {bFractionalPart=true, speedY=0.1}) {samplerType=SAMPLERTYPE_Normal, texture=/Game/Global/Textures/Default/Tileable_Patterns/2ndInfectedMapNormal}
  n19 = Add(
   A=ComponentMask(None=Add(A=@n14, B=@n15)) {g=true}
   B=DotProduct(A=ComponentMask(None=@n18.RGB) {g=true, r=true}, B=ScalarParameter() {defaultValue=0.03, parameterName=NoiseScale})
  )
  n20 = ScalarParameter() {defaultValue=7.1679, parameterName=GradientExponent}
  n21 = Constant() {r=0.5}
  n22 = SphereMask(A=OneMinus(None=Power(Base=@n19, Exp=@n20)), B=@n21, Radius=@n21, Hardness=ScalarParameter() {parameterName=HardnessA})
  n23 = Subtract(
   A=MaterialFunctionCall() {materialFunction=&ObjectPivotPoint}.Object Pivot Location
   B=WorldPosition().XYZ
  )
  n24 = Saturate(None=DotProduct(A=Transform(None=@n18.RGB), B=Normalize(VectorInput=@n23)))
  n25 = Power(Base=@n22, Exp=ScalarParameter() {defaultValue=8, parameterName=EmissiveLerpExponent})
  n26 = LinearInterpolate(
   A=LinearInterpolate(
    A=VectorParameter() {parameterName=RingEmissiveA, defaultValue=<rgba 0.0237 0.2746 0.4 0>}.RGB
    B=VectorParameter() {parameterName=StrandsEmissive, defaultValue=<rgba 0.3354 1.1436 2.5 0>}.RGB
    Alpha=@n13
   )
   B=VectorParameter() {parameterName=RingEmissiveB, defaultValue=<rgba 0.2515 0.98 2 0>}.RGB
   Alpha=Add(A=Multiply(A=@n22, B=@n24), B=@n25)
  )
  n27 = SphereMask(
   A=TextureCoordinate()
   B=Constant() {r=0.5}
   Radius=ScalarParameter() {defaultValue=0.4218, parameterName=ImageRadius}
   Hardness=ScalarParameter() {defaultValue=0.8675, parameterName=ImageHardness}
  )
  n28 = Add(A=@n22, B=@n13)
  n29 = Saturate(None=Subtract(A=@n27, B=Saturate(None=@n28)))
  LinearInterpolate(A=@n26, Alpha=@n29) {constB=0}
  n30 = Multiply(
   A=Add(A=TextureCoordinate() {coordinateIndex=1}, B=@n9)
   B=AppendVector(A=ScalarParameter() {defaultValue=4, parameterName=OuterRingTileUb}, B=ScalarParameter() {defaultValue=0.24, parameterName=OuterRingTileVb})
  )
  n31 = AppendVector(A=ScalarParameter() {parameterName=OuterRingPanSpeedUb}, B=ScalarParameter() {defaultValue=0.0675, parameterName=OuterRingPanSpeedVb})
  n32 = Multiply(
   A=Add(A=TextureCoordinate() {coordinateIndex=1}, B=@n9)
   B=AppendVector(A=ScalarParameter() {defaultValue=2, parameterName=OuterRingTileUa}, B=ScalarParameter() {defaultValue=0.24, parameterName=OuterRingTileVa})
  )
  n33 = AppendVector(A=ScalarParameter() {parameterName=OuterRingPanSpeedUa}, B=ScalarParameter() {defaultValue=0.04, parameterName=OuterRingPanSpeedVa})
  n34 = Add(
   A=TextureSample(UVs=Panner(Coordinate=@n30, Speed=@n31) {bFractionalPart=true, speedY=0.1}) {texture=/Game/Weapons/FORT_Melee/Effects/LaserSwordFounders/T_LaserSwordTrailNoise}.RGB
   B=TextureSample(UVs=Panner(Coordinate=@n32, Speed=@n33) {bFractionalPart=true, speedY=0.1}) {samplerType=SAMPLERTYPE_LinearColor, texture=/Game/Characters/Player/Male/Medium/Bodies/M_MED_SRDrift/Textures/T_ElectricNoise}.RGB
  )
  n35 = Multiply(A=TextureCoordinate() {coordinateIndex=1}, B=ScalarParameter() {defaultValue=3, parameterName=OuterScale})
  n36 = Multiply(
   A=Power(Base=Divide(A=@n34), Exp=ScalarParameter() {defaultValue=1.5, parameterName=OuterRingPower})
   B=SphereMask(
    A=ComponentMask(None=@n35) {g=true}
    B=ScalarParameter() {defaultValue=2.1802, parameterName=OuterPosition}
    Radius=ScalarParameter() {defaultValue=0.7482, parameterName=OuterRadius}
    Hardness=Constant()
   )
  )
  n37 = Multiply(
   A=TextureCoordinate() {coordinateIndex=1}
   B=AppendVector(A=ScalarParameter() {defaultValue=2, parameterName=CelestialTileU}, B=ScalarParameter() {defaultValue=0.5, parameterName=CelestialTileV})
  )
  n38 = Panner(Coordinate=BumpOffset(Coordinate=Add(A=@n37, B=@n9), Height=ScalarParameter() {defaultValue=4, parameterName=CelestialHeight})) {bFractionalPart=true, speedY=0.05}
  n39 = Add(
   A=Multiply(
    A=TextureSample(UVs=@n38) {texture=/Game/Characters/Player/Male/Medium/Bodies/M_Med_Soldier_04/Skins/Galaxy/Textures/T_FN_Nebula}.RGB
    B=VectorParameter() {parameterName=CelestialColorScale, defaultValue=<rgba 5 5 5 0>}.RGB
   )
   B=VectorParameter() {parameterName=CelestialColorBias, defaultValue=<rgba 0.0225 0.25 0.4184 0>}.RGB
  )
  n40 = TextureCoordinate()
  n41 = Constant() {r=0.5}
  n42 = ScalarParameter() {defaultValue=2, parameterName=BumpOffsetRadius}
  n43 = ScalarParameter() {defaultValue=0.3, parameterName=BumpOffsetHardness}
  n44 = BumpOffset(
   Coordinate=MaterialFunctionCall(Texture Scale=ScalarParameter() {defaultValue=0.85, parameterName=Texture Scale}) {materialFunction=&ScaleUVsByCenter}.UVs
   Height=Multiply(A=ScalarParameter() {defaultValue=-5, parameterName=Height}, B=SphereMask(A=@n40, B=@n41, Radius=@n42, Hardness=@n43))
  )
  LinearInterpolate(
   A=Multiply(A=VectorParameter() {parameterName=OuterEdgeColor, defaultValue=<rgba 0.6007 0.2915 0.9487 0>}.RGB, B=@n36)
   B=Add(
    A=Multiply(A=@n12, B=@n39)
    B=TextureSampleParameter2D(UVs=@n44) {parameterName=Texture, texture=&/Cops_and_Robbers_512}.RGB
   )
   Alpha=@n29
  )
  LinearInterpolate(A=@n18.RGB, B=VertexNormalWS(), Alpha=@n27)
  LinearInterpolate(A=@n26, B=VectorParameter() {parameterName=HotInnerEdgeEmissive}.RGB, Alpha=@n25)
  n45 = VectorParameter() {parameterName=CurlPanSpeed, defaultValue=<rgba 10 1 0 0>}
  n46 = Time()
  n47 = Divide(A=Add(A=WorldPosition().XYZ, B=Multiply(A=@n45.RGB, B=@n46)), B=ScalarParameter() {defaultValue=120, parameterName=CurlTileScale})
  n48 = TextureCoordinate() {coordinateIndex=1}
  Multiply(
   A=Multiply(A=VectorNoise(World Position=@n47) {noiseFunction=VNF_CurlALU}, B=ScalarParameter() {defaultValue=2, parameterName=WPOScale})
   B=SphereMask(A=ComponentMask(None=@n48) {g=true}, B=Constant() {r=0.7}, Radius=Constant() {r=0.55}, Hardness=Constant() {r=0.2})
  )
  Saturate(None=Add(A=Add(A=@n27, B=@n22), B=@n36))
 ]
}
