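# Unlit alpha-composite material for an island thumbnail: a texture image inside a soft, smoky, tinted frame.
# Two panning smoke samples distort the center-scaled image UVs and break up a rounded-square mask built from powered, centered UVs.
# That mask plus a noisy sphere-mask ring drives opacity and blends tint to image, with color premultiplied by opacity.
# A synced distance fade both reshapes the frame mask and scales final opacity.
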Material /Game/Playgrounds/Items/M_IslandThumbail
graph {
 properties {blendMode=BLEND_AlphaComposite, pixelDepthOffsetMode=PDOM_Legacy, shadingModel=MSM_Unlit, usedShadingModels=Unlit}
 nodes [
  MaterialFunctionCall!(
   WorldPosition=WorldPosition().XYZ
   PivotLocation
   WS Normals
   Custom Object Basis 1=Constant3Vector() {constant=<rgba 0 1 0 0>}
   Custom Object Basis 2=Constant3Vector() {constant=<rgba 1 0 0 0>}
   Custom Object Basis 3=Constant3Vector() {constant=<rgba 0 0 1 0>}
  ) {materialFunction=/Engine/Functions/Engine_MaterialFunctions02/WorldPositionOffset/AlignMeshToTheCamera}
  Multiply()
  Multiply(A=Subtract(A=ObjectPositionWS(), B=WorldPosition().XYZ)) {constB=0.5}
  n1 = VectorParameter() {parameterName=Color, defaultValue=<rgba 0.1929 0.8314 3 0>}
  n2 = TextureCoordinate()
  n3 = ScalarParameter() {defaultValue=0.81, parameterName=CenterScale}
  n4 = TextureCoordinate()
  n5 = VectorParameter() {parameterName=TilePanA, defaultValue=<rgba 1.344 0.8064 0.05 0.0433>}
  n6 = AppendVector(A=@n5.R, B=@n5.R)
  n7 = TextureSample(UVs=Panner(Coordinate=Multiply(A=@n4, B=@n6), Speed=AppendVector(A=@n5.B, B=@n5.B)) {speedX=0.1, speedY=0.1}) {texture=&T_Inky_Smoke_Tile}
  n8 = TextureCoordinate()
  n9 = VectorParameter() {parameterName=TilePanB, defaultValue=<rgba 1.68 1.008 -0.04 -0.0532>}
  n10 = AppendVector(A=@n9.R, B=@n9.R)
  n11 = TextureSample(UVs=Panner(Coordinate=Multiply(A=@n8, B=@n10), Speed=AppendVector(A=@n9.B, B=@n9.B)) {speedX=0.1, speedY=0.1}) {texture=&T_Inky_Smoke_Tile}
  n12 = Subtract(A=AppendVector(A=@n7.R, B=@n11.R)) {constB=0.5}
  n13 = Multiply(A=@n12, B=ScalarParameter() {defaultValue=0.06, parameterName=UVDistortScale})
  n14 = ScalarParameter() {defaultValue=0.15, parameterName=NoiseScale}
  n15 = TextureCoordinate()
  n16 = MaterialFunctionCall(
   Fade_Start_Distance=ScalarParameter() {defaultValue=5000, parameterName="Fade_Start_Distance (S)"}
   Fade_End_Distance=ScalarParameter() {defaultValue=4800, parameterName="Fade_End_Distance (S)"}
  ) {materialFunction=&MF_Distance_SyncedFade_FX}
  n17 = LinearInterpolate(B=Abs(None=ConstantBiasScale(None=@n15) {bias=-0.5, scale=2}), Alpha=@n16.Result Inverted) {constA=1}
  n18 = ScalarParameter() {defaultValue=3.25, parameterName=FrameExponent}
  n19 = DotProduct(A=Add(A=Multiply(A=@n12, B=@n14), B=Power(Base=@n17, Exp=@n18)), B=Constant() {r=1})
  n20 = OneMinus(None=Saturate(None=@n19))
  n21 = ScalarParameter() {defaultValue=1.175, parameterName=LerpScale}
  n22 = Saturate(None=Power(Base=Multiply(A=@n20, B=@n21), Exp=ScalarParameter() {defaultValue=3.5, parameterName=LerpExponent}) {constExponent=5})
  n23 = TextureSampleParameter2D(UVs=Add(
    A=MaterialFunctionCall(UVs=@n2, Texture Scale=@n3) {materialFunction=&ScaleUVsByCenter}.UVs
    B=Multiply(A=@n13, B=@n22)
   )) {parameterName=Texture, texture=&T_WhitesRGB}
  n24 = Add(A=@n7.R, B=@n11.R)
  n25 = Multiply(A=Multiply(A=@n24, B=@n1.RGB), B=OneMinus(None=@n22))
  n26 = Multiply(
   A=@n24
   B=SphereMask(
    A=@n19
    B=ScalarParameter() {defaultValue=0.4, parameterName=OuterRingCenter}
    Radius=ScalarParameter() {defaultValue=0.75, parameterName=OuterRingRadius}
    Hardness=ScalarParameter() {defaultValue=0.15, parameterName=OuterRingHardness}
   )
  )
  n27 = Power(Base=Saturate(None=Add(A=@n22, B=@n26)), Exp=ScalarParameter() {defaultValue=5, parameterName=OpacityExponent})
  Multiply(A=LinearInterpolate(A=@n1.RGB, B=Add(A=@n23.RGB, B=@n25), Alpha=@n27), B=@n27)
  Multiply(A=@n27, B=@n16.Result Inverted)
 ]
}
