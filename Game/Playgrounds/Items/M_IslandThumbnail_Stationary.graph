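# Unlit, two-sided alpha-composite material that displays a texture, named as the stationary variant of an island thumbnail for Playgrounds items.
# A CenterScale parameter scales the UVs around their center before sampling the Texture parameter, which defaults to a white placeholder.
# The sampled RGB sets Emissive Color through material attributes, which then pass through MF_CreativeMoveTool, presumably for Creative move-tool visuals.
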Material /Game/Playgrounds/Items/M_IslandThumbnail_Stationary
graph {
 properties {
  bUseMaterialAttributes=true
  blendMode=BLEND_AlphaComposite
  pixelDepthOffsetMode=PDOM_Legacy
  shadingModel=MSM_Unlit
  twoSided=true
  usedShadingModels=Unlit
 }
 nodes [
  n1 = MaterialFunctionCall(UVs=TextureCoordinate(), Texture Scale=ScalarParameter() {defaultValue=1, parameterName=CenterScale}) {materialFunction=&ScaleUVsByCenter}
  MaterialFunctionCall(MA In=SetMaterialAttributes(Emissive Color=TextureSampleParameter2D(UVs=@n1.UVs) {parameterName=Texture, texture=&T_WhitesRGB}.RGB) {attributeSetTypes [#1]}) {materialFunction=&MF_CreativeMoveTool}
 ]
}
